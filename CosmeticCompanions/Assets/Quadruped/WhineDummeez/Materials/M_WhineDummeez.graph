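# Skinned character material for the WhineDummeez quadruped cosmetic companion, adding color customization and masked clear coat between the shared CharacterParent 2023 PreFX and PostFX functions.
# A channel mask applies CustomColor over the base color's red channel, modulated by detail diffuse.
# A tech-art mask channel blends in a clear coat shading model with Fresnel-tinted color, noise-driven coat roughness and optional metallic, specular and roughness overrides.
# Tiled flake normals feed ClearCoatNormalCustomOutput.
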Material /CosmeticCompanions/Assets/Quadruped/WhineDummeez/Materials/M_WhineDummeez
graph {
 properties {
  bUseMaterialAttributes=true
  bUsedWithClothing=true
  bUsedWithMeshDeformer=true
  bUsedWithMorphTargets=true
  bUsedWithSkeletalMesh=true
  pixelDepthOffsetMode=PDOM_Legacy
  shadingModel=MSM_FromMaterialExpression
  usedShadingModels=Default Lit | Subsurface | Clear Coat | Subsurface Profile | Cloth
 }
 nodes [
  n1 = NamedRerouteUsage()
  n2 = ScalarParameter() {defaultValue=10, group=06_AdvancedClearCoat_v2, parameterName=DetailNormalTiling, sortPriority=14}
  n3 = TextureSampleParameter2D(UVs=NamedRerouteDeclaration(None=Multiply(A=@n1, B=@n2)) {name=ClearCoatNormalUVs, nodeColor=<rgba 0.0547 0 1 1>}) {
   group=06_AdvancedClearCoat_v2
   parameterName=DetailNormal
   samplerType=SAMPLERTYPE_Normal
   sortPriority=13
   texture=/Game/Environments/Sets/Neo/Vehicles/Textures/T_Flakes_N
  }
  n4 = LinearInterpolate(Alpha=ScalarParameter() {defaultValue=0.5, group=06_AdvancedClearCoat_v2, parameterName=DetailNormalIntensity, sortPriority=15}) {constA=1, constB=0}
  ClearCoatNormalCustomOutput!(None=MaterialFunctionCall(
    BaseNormal=NamedRerouteUsage()
    AdditionalNormal=MaterialFunctionCall(Normal=@n3.RGB, Flatness=@n4) {materialFunction=&FlattenNormal}.Result
   ) {materialFunction=&BlendAngleCorrectedNormals}.Result)
  n5 = NamedRerouteUsage()
  n6 = MaterialFunctionCall() {materialFunction=&MF_CharacterParent_2023_PreFX}
  MaterialFunctionCall(Material=@n5, M=@n6.M, Pre Skinned Local Position=@n6.M, Distance Blend=@n6.M, Fresnel=@n6.M) {materialFunction=&MF_CharacterParent_2023_PostFX}
  NamedRerouteDeclaration(None=@n6.Result) {name=Material_IN, nodeColor=<rgba 1 0 0 1>}
  NamedRerouteDeclaration(None=TextureCoordinate()) {name=UVs, nodeColor=<rgba 1 0.2578 0 1>}
  n7 = NamedRerouteUsage()
  NamedRerouteDeclaration(None=Power(
    Base=Abs(None=@n7)
    Exp=ScalarParameter() {defaultValue=2, group=06_AdvancedClearCoat_v2, parameterName=ClearCoat_FresnelPower, sortPriority=5}
   ) {constExponent=5}) {name=Fresnel, nodeColor=<rgba 0.3125 0 1 1>}
  n8 = ScalarParameter() {group=06_AdvancedClearCoat_v2, parameterName=ClearCoat_RoughnessMin, sortPriority=10}
  n9 = ScalarParameter() {defaultValue=1, group=06_AdvancedClearCoat_v2, parameterName=ClearCoat_RoughnessMax, sortPriority=11}
  n10 = TextureSampleParameter2D(UVs=Multiply(
    A=NamedRerouteUsage()
    B=ScalarParameter() {defaultValue=5, group=06_AdvancedClearCoat_v2, parameterName=ClearCoat_RoughnessTiling, sortPriority=9}
   )) {
   group=06_AdvancedClearCoat_v2
   parameterName=ClearCoat_Roughness
   samplerType=SAMPLERTYPE_Masks
   sortPriority=7
   texture=/Game/Global/Textures/Default/Noise/T_Tiling_Grime
  }
  n11 = ChannelMaskParameter(None=@n10.RGB) {group=06_AdvancedClearCoat_v2, maskChannel=Blue, parameterName=ClearCoat_RoughnessChannel, sortPriority=8, defaultValue=<rgba 0 0 1 0>}
  n12 = LinearInterpolate(
   A=@n11.R
   B=NamedRerouteUsage()
   Alpha=ScalarParameter() {group=06_AdvancedClearCoat_v2, parameterName=ClearCoat_BaseRoughnessInfluence, sortPriority=12}
  )
  NamedRerouteDeclaration(None=Saturate(None=LinearInterpolate(A=@n8, B=@n9, Alpha=@n12))) {name=ClearCoatRoughness, nodeColor=<rgba 0 0.0625 1 1>}
  n13 = ScalarParameter() {defaultValue=1, group=06_AdvancedClearCoat_v2, parameterName=ClearCoat_BaseColorBrightness, sortPriority=6}
  n14 = NamedRerouteUsage()
  n15 = VectorParameter() {group=06_AdvancedClearCoat_v2, parameterName=ClearCoat_FresnelColor, sortPriority=5, defaultValue=<rgba 0.2188 0.3082 1 1>}
  n16 = NamedRerouteUsage()
  n17 = ScalarParameter() {defaultValue=1, group=06_AdvancedClearCoat_v2, parameterName=ClearCoat_BaseColorFresnelStrength, sortPriority=6}
  NamedRerouteDeclaration(None=LinearInterpolate(A=Multiply(A=@n13, B=@n14), B=@n15.RGB, Alpha=Multiply(A=@n16, B=@n17))) {name=ClearCoatBaseColor, nodeColor=<rgba 0.3828 1 0 1>}
  n18 = GetMaterialAttributes(None=NamedRerouteUsage()) {attributeGetTypes [
    #1
    #2
    #3
    #4
    #5
   ]}
  NamedRerouteDeclaration(None=@n18.Base Color) {name=BaseColor_IN, nodeColor=<rgba 1 0.4922 0 1>}
  NamedRerouteDeclaration(None=@n18.Roughness) {name=Roughness_IN, nodeColor=<rgba 0.1719 0 1 1>}
  NamedRerouteDeclaration(None=ChannelMaskParameter(None=NamedRerouteUsage()) {group=05_Tech Art Masks, parameterName=AdvancedClearCoatChannel, sortPriority=2}.R) {name=ClearCoatChannel, nodeColor=<rgba 0.1484 1 0 1>}
  NamedRerouteDeclaration(None=@n18.Normal) {name=Normal_IN, nodeColor=<rgba 1 0 0.7266 1>}
  NamedRerouteDeclaration(None=StaticBoolParameter() {group=06_AdvancedClearCoat_v2, parameterName=ClearCoat_OverrideSRM, sortPriority=1}) {name=OverrideSRM, nodeColor=<rgba 0.3828 1 0 1>}
  NamedRerouteDeclaration(None=@n6.Tech Art Mask) {name=TechArtMask, nodeColor=<rgba 1 0.375 0 1>}
  NamedRerouteDeclaration(None=@n6.Fresnel) {name=Fresnel_IN, nodeColor=<rgba 0.2422 1 0 1>}
  n19 = ChannelMaskParameter(None=NamedRerouteUsage()) {group=06_AdvancedClearCoat_v2, maskChannel=Blue, parameterName=DetailDiffuseOverlayChannel, sortPriority=14, defaultValue=<rgba 0 0 1 0>}
  NamedRerouteDeclaration(None=@n19.R) {name=DetailDiffuse, nodeColor=<rgba 1 0.4922 0 1>}
  NamedRerouteDeclaration(None=VectorParameter() {group=CustomColor, parameterName=CustomColor, defaultValue=<rgba 1 1 1 1>}.RGB) {name=CustomColor, nodeColor=<rgba 0 1 0.7188 1>}
  n20 = NamedRerouteUsage()
  n21 = Multiply(A=ComponentMask(None=@n20) {r=true}, B=NamedRerouteUsage())
  n22 = LinearInterpolate(B=NamedRerouteUsage(), Alpha=ScalarParameter() {group=06_AdvancedClearCoat_v2, parameterName=DetailDiffuseStrength}) {constA=1}
  NamedRerouteDeclaration(None=LinearInterpolate(A=@n20, B=Multiply(A=@n21, B=@n22), Alpha=NamedRerouteUsage())) {name=BaseColorCustom, nodeColor=<rgba 1 0.3984 0 1>}
  NamedRerouteDeclaration(None=ChannelMaskParameter(None=NamedRerouteUsage()) {group=CustomColor, parameterName=ColorCustomizationChannel}.R) {name=ColorCustomization Mask, nodeColor=<rgba 1 0.5156 0 1>}
  n23 = NamedRerouteUsage()
  n24 = NamedRerouteUsage()
  n25 = SetMaterialAttributes!(
   MaterialAttributes=@n18.MaterialAttributes
   Base Color=NamedRerouteUsage()
   Metallic=StaticSwitch(
    True=ScalarParameter() {defaultValue=1, group=06_AdvancedClearCoat_v2, parameterName=ClearCoat_OverrideMetallic, sortPriority=4}
    False=@n18.Metallic
    Value=NamedRerouteUsage()
   )
   Specular=StaticSwitch(
    True=ScalarParameter() {defaultValue=1, group=06_AdvancedClearCoat_v2, parameterName=ClearCoat_OverrideSpecular, sortPriority=4}
    False=@n18.Specular
    Value=NamedRerouteUsage()
   )
   Roughness=StaticSwitch(
    True=ScalarParameter() {defaultValue=0.5, group=06_AdvancedClearCoat_v2, parameterName=ClearCoat_OverrideRoughness, sortPriority=4}
    False=@n18.Roughness
    Value=NamedRerouteUsage()
   )
   Clear Coat or Cloth or Curvature or Custom Data 0=ScalarParameter() {defaultValue=1, group=06_AdvancedClearCoat_v2, parameterName=ClearCoat_Strength}
   Clear Coat Roughness or Custom Data 1=NamedRerouteUsage()
   Shading Model=ShadingModel() {shadingModel=MSM_ClearCoat}
  ) {attributeSetTypes [
    #2
    #4
    #5
    #1
    #6
    #7
    #8
   ]}
  NamedRerouteDeclaration(None=BlendMaterialAttributes(
    A=SetMaterialAttributes!(MaterialAttributes=@n23, Base Color=@n24) {attributeSetTypes [#2]}
    B=@n25
    Alpha=NamedRerouteUsage()
   )) {name=Material_OUT, nodeColor=<rgba 0.0781 0 1 1>}
  NamedRerouteUsage()
  NamedRerouteUsage()
  NamedRerouteUsage()
  NamedRerouteUsage()
  SetMaterialAttributes(MaterialAttributes=@n5, Base Color=NamedRerouteUsage()) {attributeSetTypes [#2]}
  TextureSampleParameter2D(UVs=NamedRerouteUsage()) {group=06_AdvancedClearCoat_v2, parameterName=DetailDiffuse, texture=&T_WhitesRGB}
 ]
}
